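; Parent passive gameplay ability that marks the closest terminal for a player who has the keycard equipped.
; On activation, if the terminal-indicator hotfix is enabled, it adds the mark gameplay cue to the owner.
; The cue parameters carry hotfix settings in repurposed fields: update frequency as RawMagnitude and the direction-FX toggle as AbilityLevel.
; Activation and failed passive activation are both written to the shipping log.
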(event Ability|EventActivateAbility
  ;[0.0568 0.2344 0.02719] ActivateAbility: When the keycard is equipped, apply the gameplay cue on the player that marks the closest terminal for them.
  (|CallParentFunction)
  (Fortnite|FortShippingLog self (Utilities|String|Append "ActivateAbility: Ability activated for player:" (Fortnite|GetPlayerNameSafe (CallFunction|GetAvatarasFortPlayerPawn))))
  (if (ScalableFloat|GetValueasBool (Variables|Default|GetHotfix_EnableTerminalIndicator))
    (Ability|AddGameplayCueWithParamsToOwner (Variables|Default|GetTerminalKeycardMarkGameplayCueTag) (Ability|GameplayCue|MakeGameplayCueParameters :NormalizedMagnitude 0.0 :RawMagnitude (ScalableFloat|GetValueatLevel (Variables|Default|GetHotfix_TerminalIndicatorUpdateFrequency)) :Location "0,0,0" :Normal "0,0,0" :GameplayEffectLevel 1 :AbilityLevel (ScalableFloat|GetValueasInteger (Variables|Default|GetHotfix_AllowTerminalDirectionFXIndicator)) :bReplicateLocationWhenUsingMinimalRepProxy false))))

(event Ability|EventFailedToActivatePassiveAbility
  (params (ActorInfo :type Gameplay%20Ability%20Actor%20Info%20Structure))
  ;[1 0 0] Failed To Activate Passive Ability: Log the failure.
  (|CallParentFunction ActorInfo)
  (Fortnite|FortShippingLog self "Failed To Activate Passive Ability"))
